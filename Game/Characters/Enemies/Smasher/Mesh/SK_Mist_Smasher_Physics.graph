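# Physics asset for the Mist Smasher enemy's skeletal mesh, defining its collision bodies and joint constraints for physics simulation.
# Box, capsule and sphere bodies cover hips, spine, head and legs without feet, with compound shapes on bulkier torso bones.
# The arms are asymmetric: a three-body right arm, and an eight-segment left chain flaring into a heavy sphere-tipped end.
# Knees and right elbow act as single-axis hinges, and thighs lock twist.
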PhysicsAsset /Game/Characters/Enemies/Smasher/Mesh/SK_Mist_Smasher_Physics
bodies {
 b_Arm_1_L [{shapeType=Capsule, length=8.8492, radius=15.4905, center=<11.0396 1.4351 0.9924>, rotation=<rot 0 90 -90>}]
 b_Arm_2_L [{shapeType=Capsule, length=8.8558, radius=13.0553, center=<19.5065 -1.4921 -0.9312>, rotation=<rot 0 90 -85>}]
 b_Arm_3_L [{shapeType=Capsule, length=11.97, radius=12.8386, center=<14.8202 -1.5874 0.073>, rotation=<rot 0 90 -90>}]
 b_Arm_4_L [{shapeType=Capsule, length=10.9083, radius=12.8139, center=<11.9398 -1.2146 0.9095>, rotation=<rot 0 90 -90>}]
 b_Arm_5_L [{shapeType=Capsule, length=11.0897, radius=16.607, center=<16.9956 -1.2951 2.3081>, rotation=<rot 0 90 -90>}]
 b_Arm_6_L [{shapeType=Capsule, length=9.0993, radius=23.7056, center=<20.38 -0.2981 1.2457>, rotation=<rot 0 90 -90>}]
 b_Arm_7_L [{shapeType=Capsule, length=11.1571, radius=27.5535, center=<19.1006 0.2432 1.4855>, rotation=<rot 0 90 -90>}]
 b_Arm_8_L [{shapeType=Sphere, radius=29.1156, center=<7.71 1.2798 -0.085>}]
 b_Calf_L [{shapeType=Box, extentX=42.8227, extentY=12.74, extentZ=12.1113, center=<26.5334 -0.468 1.0568>}]
 b_Calf_R [{shapeType=Box, extentX=38.337, extentY=12.6953, extentZ=12.0879, center=<-26.03 0.4598 -0.8>}]
 b_Head [{shapeType=Capsule, length=23.0723, radius=20.1331, center=<-6.25 6.5 0>, rotation=<rot -90 0 0>}]
 b_Hips [
  {shapeType=Sphere, radius=25.5703, center=<18.5 -8.25 -3.5>}
  {shapeType=Sphere, radius=23.4165, center=<-15.75 -14.5 -3.5>}
  {shapeType=Capsule, length=25.1592, radius=21.5, center=<0 -10.25 -4.25>, rotation=<rot -90 0 0>}
 ]
 b_LowerArm_R [{shapeType=Capsule, length=17.3949, radius=13.8427, center=<18.4787 0.7798 -0.578>, rotation=<rot 0 91.626 -84.9855>}]
 b_Spine_01 [
  {shapeType=Box, extentX=37.9326, extentY=50.5009, extentZ=37.2433, center=<10.9571 -4.9273 0.2799>, rotation=<rot 0 83.7586 -90>}
  {shapeType=Box, extentX=27.5841, extentY=26.5484, extentZ=63.498, center=<-11.75 12.75 0>}
 ]
 b_Spine_02 [{shapeType=Box, extentX=69.4934, extentY=69.1063, extentZ=29.2238, center=<-1.5659 -2.3162 1.383>, rotation=<rot 0 89.0514 -90>}]
 b_Spine_03 [
  {shapeType=Sphere, radius=32.2232, center=<-11.25 -0.5 20.75>}
  {shapeType=Sphere, radius=33.3384, center=<8.25 -9 -17.75>}
  {shapeType=Sphere, radius=22.3571, center=<-12.75 17.75 -18.5>}
  {shapeType=Sphere, radius=26.7437, center=<1.5 0 -32.25>}
  {shapeType=Sphere, radius=29.6534, center=<-14.75 -9.75 -12.5>}
 ]
 b_Thigh_L [{shapeType=Box, extentX=48.87, extentY=20.6848, extentZ=20.7904, center=<20.6518 3.084 -0.0726>}]
 b_Thigh_R [{shapeType=Box, extentX=48.8086, extentY=22.8764, extentZ=17.1835, center=<-20.7003 -2.3517 0.0915>}]
 b_UpperArm_R [{shapeType=Capsule, length=25.7469, radius=14.0982, center=<17.2356 -0.3313 3.4432>, rotation=<rot 0 89.1655 -88.3574>}]
 b_Wrist [{shapeType=Box, extentX=28.2787, extentY=19.541, extentZ=29.2325, center=<12.0029 0.1083 -0.1038>}]
}
constraints [
 {bone1Name=b_Spine_01, bone2Name=b_Hips, swing1LimitDegrees=30, swing2LimitDegrees=30, twistLimitDegrees=30}
 {bone1Name=b_Spine_02, bone2Name=b_Spine_01, swing1LimitDegrees=30, swing2LimitDegrees=30, twistLimitDegrees=30}
 {bone1Name=b_Spine_03, bone2Name=b_Spine_02, swing1LimitDegrees=30, swing2LimitDegrees=30, twistLimitDegrees=30}
 {bone1Name=b_Head, bone2Name=b_Spine_03, swing1LimitDegrees=30, swing2LimitDegrees=30, twistLimitDegrees=30}
 {bone1Name=b_Thigh_L, bone2Name=b_Hips, swing1LimitDegrees=60, swing2LimitDegrees=30, twistLimitDegrees=45, twistMotion=Locked}
 {
  bone1Name=b_Calf_L
  bone2Name=b_Thigh_L
  swing1LimitDegrees=60
  swing2LimitDegrees=45
  swing2Motion=Locked
  twistLimitDegrees=45
  twistMotion=Locked
 }
 {bone1Name=b_Thigh_R, bone2Name=b_Hips, swing1LimitDegrees=60, swing2LimitDegrees=30, twistLimitDegrees=45, twistMotion=Locked}
 {
  bone1Name=b_Calf_R
  bone2Name=b_Thigh_R
  swing1LimitDegrees=60
  swing2LimitDegrees=45
  swing2Motion=Locked
  twistLimitDegrees=45
  twistMotion=Locked
 }
 {bone1Name=b_UpperArm_R, bone2Name=b_Spine_03, swing1LimitDegrees=30, swing2LimitDegrees=30, twistLimitDegrees=30}
 {
  bone1Name=b_LowerArm_R
  bone2Name=b_UpperArm_R
  swing1LimitDegrees=60
  swing2LimitDegrees=45
  swing2Motion=Locked
  twistLimitDegrees=45
  twistMotion=Locked
 }
 {bone1Name=b_Wrist, bone2Name=b_LowerArm_R, swing1LimitDegrees=30, swing2LimitDegrees=30, twistLimitDegrees=30}
 {bone1Name=b_Arm_1_L, bone2Name=b_Spine_03, swing1LimitDegrees=30, swing2LimitDegrees=30, twistLimitDegrees=30}
 {bone1Name=b_Arm_2_L, bone2Name=b_Arm_1_L, swing1LimitDegrees=30, swing2LimitDegrees=30, twistLimitDegrees=30}
 {bone1Name=b_Arm_3_L, bone2Name=b_Arm_2_L, swing1LimitDegrees=30, swing2LimitDegrees=30, twistLimitDegrees=30}
 {bone1Name=b_Arm_4_L, bone2Name=b_Arm_3_L, swing1LimitDegrees=30, swing2LimitDegrees=30, twistLimitDegrees=30}
 {bone1Name=b_Arm_5_L, bone2Name=b_Arm_4_L, swing1LimitDegrees=30, swing2LimitDegrees=30, twistLimitDegrees=30}
 {bone1Name=b_Arm_6_L, bone2Name=b_Arm_5_L, swing1LimitDegrees=30, swing2LimitDegrees=30, twistLimitDegrees=30}
 {bone1Name=b_Arm_7_L, bone2Name=b_Arm_6_L, swing1LimitDegrees=30, swing2LimitDegrees=30, twistLimitDegrees=30}
 {bone1Name=b_Arm_8_L, bone2Name=b_Arm_7_L, swing1LimitDegrees=30, swing2LimitDegrees=30, twistLimitDegrees=30}
]
