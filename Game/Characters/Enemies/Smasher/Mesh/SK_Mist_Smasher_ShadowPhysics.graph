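# Shadow physics asset for the Smasher enemy's skeletal mesh, approximating the character with one capsule per major bone, as the name suggests for capsule shadows.
# Capsules cover hips, two spine bones, head, both legs and the right arm, plus a five-segment chain for the asymmetric left arm that widens sharply toward its end.
# Constraints link the bodies in a tree from the hips, with all swing and twist motion free.
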PhysicsAsset /Game/Characters/Enemies/Smasher/Mesh/SK_Mist_Smasher_ShadowPhysics
bodies {
 b_Arm_1_L [{shapeType=Capsule, length=19.0671, radius=15.497, center=<11.3296 1.512 -0.9373>, rotation=<rot -18.88 -100.2596 92.3559>}]
 b_Arm_2_L [{shapeType=Capsule, length=17.588, radius=16.3865, center=<20.1177 0.5585 1.2474>, rotation=<rot -24.2776 91.8716 95.6912>}]
 b_Arm_4_L [{shapeType=Capsule, length=14.5583, radius=11.3825, center=<7.9437 -1.0075 0.9741>, rotation=<rot -8.9654 93.2484 -93.3992>}]
 b_Arm_6_L [{shapeType=Capsule, length=23.5197, radius=19.958, center=<9.1903 0.0145 3.8323>, rotation=<rot -4.8115 105.9421 -84.255>}]
 b_Arm_8_L [{shapeType=Capsule, length=17.7029, radius=30.0785, center=<-0.0643 3.7566 0.9167>, rotation=<rot -24.7104 -89.9999 -80.8799>}]
 b_Calf_L [{shapeType=Capsule, length=42.3594, radius=6.357, center=<24.4421 -0.4785 0.8524>, rotation=<rot 0 -90.0757 -88.0091>}]
 b_Calf_R [{shapeType=Capsule, length=42.5544, radius=6.3047, center=<-24.2778 0.5038 -0.5311>, rotation=<rot 0 -90.2516 -88.9662>}]
 b_Head [{shapeType=Capsule, length=36.9772, radius=15.9003, center=<0.6897 9.8103 1.4783>, rotation=<rot 0 70.8915 -89.9776>}]
 b_Hips [{shapeType=Capsule, length=32.5716, radius=25.7781, center=<-0.1332 -2.1646 -3.4868>, rotation=<rot 0 -90.5651 -88.1458>}]
 b_LowerArm_R [{shapeType=Capsule, length=35.0827, radius=13.3161, center=<27.5617 0.6764 3.4265>, rotation=<rot 0 -80.2095 -101.4492>}]
 b_Spine_01 [{shapeType=Capsule, length=30.909, radius=25.8745, center=<9.4615 -2.4686 0.5508>, rotation=<rot -4.9605 -90 2.3072>}]
 b_Spine_03 [{shapeType=Capsule, length=32.0766, radius=37.6539, center=<-6.1132 1.2106 -2.0916>, rotation=<rot 11.0241 11.4616 -1.5831>}]
 b_Thigh_L [{shapeType=Capsule, length=31.8613, radius=9.7736, center=<20.0022 -1.975 -0.5754>, rotation=<rot 0 -95.1751 -94.099>}]
 b_Thigh_R [{shapeType=Capsule, length=31.8472, radius=10.7632, center=<-20.4132 1.7447 -1.1052>, rotation=<rot 0 -89.3279 -87.4743>}]
 b_UpperArm_R [{shapeType=Capsule, length=42.6313, radius=9.2749, center=<17.3807 1.1437 -0.8122>, rotation=<rot 0 94.8114 -92.7366>}]
}
constraints [
 {
  bone1Name=b_Spine_01
  bone2Name=b_Hips
  swing1LimitDegrees=45
  swing1Motion=Free
  swing2LimitDegrees=45
  swing2Motion=Free
  twistLimitDegrees=45
  twistMotion=Free
 }
 {
  bone1Name=b_Head
  bone2Name=b_Spine_03
  swing1LimitDegrees=45
  swing1Motion=Free
  swing2LimitDegrees=45
  swing2Motion=Free
  twistLimitDegrees=45
  twistMotion=Free
 }
 {
  bone1Name=b_UpperArm_R
  bone2Name=b_Spine_03
  swing1LimitDegrees=45
  swing1Motion=Free
  swing2LimitDegrees=45
  swing2Motion=Free
  twistLimitDegrees=45
  twistMotion=Free
 }
 {
  bone1Name=b_LowerArm_R
  bone2Name=b_UpperArm_R
  swing1LimitDegrees=45
  swing1Motion=Free
  swing2LimitDegrees=45
  swing2Motion=Free
  twistLimitDegrees=45
  twistMotion=Free
 }
 {
  bone1Name=b_Arm_4_L
  bone2Name=b_Arm_2_L
  swing1LimitDegrees=45
  swing1Motion=Free
  swing2LimitDegrees=45
  swing2Motion=Free
  twistLimitDegrees=45
  twistMotion=Free
 }
 {
  bone1Name=b_Arm_6_L
  bone2Name=b_Arm_4_L
  swing1LimitDegrees=45
  swing1Motion=Free
  swing2LimitDegrees=45
  swing2Motion=Free
  twistLimitDegrees=45
  twistMotion=Free
 }
 {
  bone1Name=b_Thigh_L
  bone2Name=b_Hips
  swing1LimitDegrees=45
  swing1Motion=Free
  swing2LimitDegrees=45
  swing2Motion=Free
  twistLimitDegrees=45
  twistMotion=Free
 }
 {
  bone1Name=b_Calf_L
  bone2Name=b_Thigh_L
  swing1LimitDegrees=45
  swing1Motion=Free
  swing2LimitDegrees=45
  swing2Motion=Free
  twistLimitDegrees=45
  twistMotion=Free
 }
 {
  bone1Name=b_Thigh_R
  bone2Name=b_Hips
  swing1LimitDegrees=45
  swing1Motion=Free
  swing2LimitDegrees=45
  swing2Motion=Free
  twistLimitDegrees=45
  twistMotion=Free
 }
 {
  bone1Name=b_Calf_R
  bone2Name=b_Thigh_R
  swing1LimitDegrees=45
  swing1Motion=Free
  swing2LimitDegrees=45
  swing2Motion=Free
  twistLimitDegrees=45
  twistMotion=Free
 }
 {
  bone1Name=b_Arm_8_L
  bone2Name=b_Arm_6_L
  swing1LimitDegrees=45
  swing1Motion=Free
  swing2LimitDegrees=45
  swing2Motion=Free
  twistLimitDegrees=45
  twistMotion=Free
 }
 {
  bone1Name=b_Arm_1_L
  bone2Name=b_Spine_03
  swing1LimitDegrees=45
  swing1Motion=Free
  swing2LimitDegrees=45
  swing2Motion=Free
  twistLimitDegrees=45
  twistMotion=Free
 }
 {
  bone1Name=b_Arm_2_L
  bone2Name=b_Arm_1_L
  swing1LimitDegrees=45
  swing1Motion=Free
  swing2LimitDegrees=45
  swing2Motion=Free
  twistLimitDegrees=45
  twistMotion=Free
 }
 {
  bone1Name=b_Spine_03
  bone2Name=b_Spine_01
  swing1LimitDegrees=45
  swing1Motion=Free
  swing2LimitDegrees=45
  swing2Motion=Free
  twistLimitDegrees=45
  twistMotion=Free
 }
]
